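# M_NeoT_Cable_Spline is a Default Lit material for the Neo cable environment set, flagged for use on spline meshes.
# It holds three texture samples from the NeoT_Cable set: a _D map (likely diffuse), a normal map and an _S map sampled as masks.
# The graph has no math or parameter nodes, and the file shows no connections from the samples to material inputs.
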Material /Figment_S10/Environments/Sets/Neo/Cable/Materials/M_NeoT_Cable_Spline
graph {
 properties {bUsedWithSplineMeshes=true, pixelDepthOffsetMode=PDOM_Legacy, usedShadingModels=Default Lit}
 nodes [
  TextureSample() {texture=/Figment_S10/Environments/Sets/Neo/Cable/Textures/T_NeoT_Cable_D}
  TextureSample() {samplerType=SAMPLERTYPE_Normal, texture=/Figment_S10/Environments/Sets/Neo/Cable/Textures/T_NeoT_Cable_N}
  TextureSample() {samplerType=SAMPLERTYPE_Masks, texture=/Figment_S10/Environments/Sets/Neo/Cable/Textures/T_NeoT_Cable_S}
 ]
}
